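# Control Rig for the AntiPesto drone backpack that spins its four propeller bones according to the character's movement state.
# A chain of state branches sets a speed and duration vector, with later states overriding earlier ones and front end alternating between stopped and random speeds.
# Timer checks latch the next speed, which an accumulated lerp eases toward to drive a 360-degree time loop feeding propeller rotation.
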ControlRigBlueprint /BRCosmetics/Accessories/FORT_Backpacks/Backpack_F_MED_AntiPesto/Meshes/F_MED_AntiPesto_Pack_CtrlRig
graphs [
 RigVMFunctionLibrary {links [], nodes []}
 RigVMModel {nodes [
   RigUnit_BeginExecution = &RigUnit_BeginExecution()
   n1 = &RigVMFunction_ControlFlowBranch(ExecuteContext=@RigUnit_BeginExecution.ExecutePin, Condition=RigVMVariableNode() {variable=Is Idle}.Value)
   RigVMFunction_ControlFlowBranch = &RigVMFunction_ControlFlowBranch(ExecuteContext=@n1.Completed, Condition=RigVMVariableNode() {variable=Is Moving 2D}.Value)
   n2 = &RigVMFunction_ControlFlowBranch(ExecuteContext=@RigVMFunction_ControlFlowBranch.Completed, Condition=RigVMVariableNode() {variable=Is Tactical Sprinting}.Value)
   n3 = &RigVMFunction_ControlFlowBranch(ExecuteContext=@n2.Completed, Condition=RigVMVariableNode() {variable=Is Falling}.Value)
   n4 = &RigVMFunction_ControlFlowBranch(ExecuteContext=@n3.Completed, Condition=RigVMVariableNode() {variable=Is Skydive}.Value)
   n5 = &RigVMFunction_ControlFlowBranch(ExecuteContext=@n4.Completed, Condition=RigVMVariableNode() {variable=Is RailingOrZipline}.Value)
   n6 = &RigVMFunction_ControlFlowBranch(ExecuteContext=@n5.Completed, Condition=RigVMVariableNode() {variable=Is FrontEnd}.Value)
   n7 = RigVMVariableNode() {variable=Initial Duration}
   n8 = /Script/RigVM.RigVMFunction_TimeLoop(Normalize=false, Speed=1, Duration=@n7.Value) {is=struct RigVMUnitNode}
   n9 = &RigVMFunction_ControlFlowBranch(
    Condition=/Script/RigVM.RigVMFunction_MathDoubleGreaterEqual(A=@n8.Relative, B=&RigVMFunction_MathDoubleSub(A=0, B=0.1).Result) {is=struct RigVMUnitNode, template="GreaterEqual::Execute(in A,in B,out Result)"}.Result
    ExecuteContext=@n6.Completed
   )
   Multiply = &RigVMFunction_MathDoubleMul(B=0, A=@n7.Value)
   n10 = &RigVMFunction_ControlFlowBranch(
    Condition=&RigVMFunction_MathDoubleGreaterEqual(B=@Multiply.Result, A=@n8.Relative) {template="GreaterEqual::Execute(in A,in B,out Result)"}.Result
    ExecuteContext=@n9.Completed
   )
   AccumulateLerp = /Script/RigVM.RigVMFunction_AccumulateFloatLerp(
    Blend=1
    bIntegrateDeltaTime=true
    InitialValue=RigVMVariableNode() {variable=Current Duration}.Value
    TargetValue=RigVMVariableNode() {variable=Next Speed}.Value
   ) {is=struct RigVMUnitNode, template="AccumulateLerp::Execute(in TargetValue,in InitialValue,in Blend,in bIntegrateDeltaTime,out Result)"}
   RigVMFunction_TimeLoop = &RigVMFunction_TimeLoop(Duration=360, Normalize=false, Speed=@AccumulateLerp.Result)
   RigVMFunction_MathQuaternionFromEuler = &RigVMFunction_MathQuaternionFromEuler(Euler=(X=0,Y=0,Z=0), Euler.Z=@RigVMFunction_TimeLoop.Relative)
   &RigUnit_ModifyTransforms(
    ItemToModify="((Item=(Type=Bone,Name=\"base_propeller_lower_l_mstr\"),Transform=(Rotation=(X=0,Y=0,Z=0,W=1),Translation=(X=0,Y=0,Z=0),Scale3D=(X=1,Y=1,Z=1))),(Item=(Type=Bone,Name=\"base_propeller_lower_r_mstr\"),Transform=(Rotation=(X=0,Y=0,Z=0,W=1),Translation=(X=0,Y=0,Z= ...(truncated)"
    Weight=1
    WeightMinimum=0
    ExecutePin=@n10.Completed
    ItemToModify.0.Transform.Rotation=@RigVMFunction_MathQuaternionFromEuler.Result
    ItemToModify.1.Transform.Rotation=@RigVMFunction_MathQuaternionFromEuler.Result
    ItemToModify.2.Transform.Rotation=@RigVMFunction_MathQuaternionFromEuler.Result
    ItemToModify.3.Transform.Rotation=@RigVMFunction_MathQuaternionFromEuler.Result
   )
   n11 = RigVMVariableNode() {variable=Speed And Duration}
   RigVMVariableNode(ExecuteContext=@n10.True, Value=@n11.Value.X) {variable=Next Speed}
   RigVMVariableNode(ExecuteContext=@n9.False, Value=RigVMVariableNode() {variable=Initial Duration}.Value) {variable=Initial Duration}
   RigVMVariableNode(ExecuteContext=@n9.True, Value=@n11.Value.Y) {variable=Initial Duration}
   > RigVMVariableNode(Value=@n11.Value.Y) {variable=Current Duration}
   > RigVMVariableNode(Value=@n11.Value.X) {variable=Current Speed}
   GreaterEqual = &RigVMFunction_MathDoubleGreaterEqual(
    B=8
    A=&RigVMFunction_MathDoubleAdd(B=0.1, A=&RigVMFunction_TimeLoop(Duration=16, Normalize=false, Speed=1).Relative).Result
   ) {template="GreaterEqual::Execute(in A,in B,out Result)"}
   RigVMVariableNode(
    Value=(X=1000,Y=5,Z=0)
    Value.X=RigVMDispatchNode(
     False=0
     Condition=@GreaterEqual.Result
     True=/Script/RigVM.RigVMFunction_RandomFloat(Duration=10, Maximum=1000, Minimum=500, Seed=500) {is=struct RigVMUnitNode, template="Random::Execute(in Seed,in Minimum,in Maximum,in Duration,out Result)"}.Result
    ) {template="DISPATCH_RigVMDispatch_If(in Condition,in True,in False,out Result)"}.Result
    ExecuteContext=@n6.True
   ) {variable=Speed And Duration}
   RigVMVariableNode(Value=(X=1000,Y=5,Z=0), ExecuteContext=@n5.True) {variable=Speed And Duration}
   RigVMVariableNode(Value=(X=2000,Y=5,Z=0), ExecuteContext=@n4.True) {variable=Speed And Duration}
   RigVMVariableNode(Value=(X=750,Y=5,Z=0), ExecuteContext=@n3.True) {variable=Speed And Duration}
   RigVMVariableNode(Value=(X=500,Y=1,Z=0), ExecuteContext=@n2.True) {variable=Speed And Duration}
   RigVMVariableNode(Value=(X=50,Y=3,Z=0), ExecuteContext=@RigVMFunction_ControlFlowBranch.True) {variable=Speed And Duration}
   RigVMVariableNode(Value=(X=0,Y=1,Z=0), ExecuteContext=@n1.True) {variable=Speed And Duration}
   &RigVMFunction_MathDoubleSub(B=0.5, A=@n7.Value)
  ]}
]
hierarchy [
 root {loc=<0 0 0>, rot=<0 0 0 1>}
 pack {parent=root, loc=<0 -0.7316 0>, rot=<0 0 0 1>}
 base_drone_body_mstr {parent=pack, loc=<0 -2.1148 -5.7881>, rot=<0.7071 -0.7071 0 0>}
 dyn_drone_body {parent=base_drone_body_mstr, loc=<0 0 0>, rot=<0 0 0 1>}
 base_propeller_upper_l_mstr {parent=base_drone_body_mstr, loc=<7.5857 -10.5273 -13.8154>, rot=<0 0.7071 0 -0.7071>}
 dyn_propeller_upper_l {parent=base_propeller_upper_l_mstr, loc=<0 0 0>, rot=<0 0 0 1>}
 base_propeller_lower_l_mstr {parent=base_drone_body_mstr, loc=<7.5857 -10.5066 10.6587>, rot=<0 0.7071 0 -0.7071>}
 dyn_propeller_lower_l {parent=base_propeller_lower_l_mstr, loc=<0 0 0>, rot=<0 0 0 1>}
 base_propeller_lower_r_mstr {parent=base_drone_body_mstr, loc=<7.5857 10.5066 10.6587>, rot=<0 0.7071 0 -0.7071>}
 dyn_propeller_lower_r {parent=base_propeller_lower_r_mstr, loc=<0 0 0>, rot=<0 0 0 1>}
 base_propeller_upper_r_mstr {parent=base_drone_body_mstr, loc=<7.5857 10.5273 -13.8154>, rot=<0 0.7071 0 -0.7071>}
 dyn_propeller_upper_r {parent=base_propeller_upper_r_mstr, loc=<0 0 0>, rot=<0 0 0 1>}
]
variables [
 Current Duration {type=double, default=0}
 Current Speed {type=double, default=0}
 Initial Duration {type=double, default=1}
 Is Falling {type=bool, default=False}
 Is FrontEnd {type=bool, default=False}
 Is Idle {type=bool, default=False}
 Is Moving 2D {type=bool, default=False}
 Is RailingOrZipline {type=bool, default=False}
 Is Skydive {type=bool, default=False}
 Is Tactical Sprinting {type=bool, default=False}
 Next Speed {type=double, default=0}
 Speed And Duration {type=FVector, default=(X=0,Y=0,Z=0)}
]
