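# Animation Blueprint for the AntiPesto backpack cosmetic, passing the input pose through a Control Rig and an additive "pack" bone modifier.
# The Control Rig receives movement, sprint, skydive, fall, idle, zipline/rail and frontend flags.
# The event graph derives idle and zipline/rail states, tracking ziplines via anim input events.
# It also computes pack offset variables from owner spine yaw, with skydiving and frontend overrides, though the bone modifier shows no link to them.
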AnimBlueprint /BRCosmetics/Accessories/FORT_Backpacks/Backpack_F_MED_AntiPesto/Meshes/F_MED_AntiPesto_Pack_AnimBP
target_skeleton=&/F_MED_AntiPesto_Pack_Skeleton
graphs [
 AnimGraph {nodes [
   n1 = Control Rig!(
    Source=Input Pose()
    Alpha
    Is Moving 2D:bool=Get bIsMoving2D()
    Is Tactical Sprinting:bool=Get bIsTacticalSprinting()
    Is Skydive:bool=Get bIsSkydiveFloating()
    Is Falling:bool=Get bIsFalling()
    Is Idle:bool=Get isIDLE() -> (isIDLE:bool)
    Is RailingOrZipline:bool=Get isZipplingAndRailing() -> (isZipplingAndRailing:bool)
    Is FrontEnd:bool=Get bIsInFrontEnd()
   ) -> (Pose) {
    alphaBoolBlend {blendOption=Linear}
    controlRigAssetReference {blueprintRigClass=&/F_MED_AntiPesto_Pack_CtrlRig.F_MED_AntiPesto_Pack_CtrlRig_C}
   }
   n2 = "Transform (Modify) Bone - Bone: pack"!(ComponentPose=Local To Component(LocalPose=@n1), bAlphaBoolEnabled:bool, Translation="0,0,0", Rotation="0,0,0", Scale) -> (Pose) {
    alphaInputType=Bool
    rotationMode=BMM_Additive
    rotationSpace=BCS_BoneSpace
    translationMode=BMM_Additive
    translationSpace=BCS_BoneSpace
    alphaBoolBlend {blendOption=Linear}
    boneToModify {boneName=pack}
   }
   Output Pose(Result=Component To Local(ComponentPose=@n2)) {layerGroup=None}
  ]}
 EventGraph {nodes [
   Event BlueprintUpdateAnimation()
   > n1 = Sequence!(execute) -> (then_0, then_1, then_2:Exec)
   n2 = Get bIsMoving2D()
   n3 = Get bIsPlayingEmote()
   Set isIDLE(execute=@n1.then_1, isIDLE:bool=NOT Boolean(A=OR Boolean!(A=@n2, B=@n3, C:bool) -> (ReturnValue))) -> (Output_Get:bool)
   n4 = Branch(execute=@n1.then_2, Condition=OR Boolean!(A=Get isZiplining() -> (isZiplining:bool), B=Get bIsOnGrindRail()) -> (ReturnValue))
   > Set isZipplingAndRailing(isZipplingAndRailing:bool=true) -> (Output_Get:bool)
   Set isZipplingAndRailing(execute=@n4.else, isZipplingAndRailing=false)
   n5 = Sequence!(execute=@n1.then_0) -> (then_0, then_1)
   > n6 = Branch(Condition=Get rotateback() -> (rotateback:bool))
   n7 = GetOwnerSkeletalMeshComponent()
   n8 = BreakRotator(InRot=BreakTransform(InTransform=GetDeltaTransformFromRefPose(self=@n7, BoneName=spine_05, BaseName=spine_01)).Rotation)
   n9 = MapRangeClamped(Value=@n8.Yaw, InRangeA=-10, InRangeB=-30, OutRangeB=-10)
   n10 = MapRangeClamped(Value=@n8.Yaw, InRangeA=-10, InRangeB=-30, OutRangeB=0)
   > Set backpack_rotate(backpack_rotate=MakeRotator(Roll=MapRangeClamped(Value=@n8.Yaw, InRangeA=-10, InRangeB=-30, OutRangeA=10, OutRangeB=20), Pitch=@n10, Yaw=@n9))
   Set backpack_rotate(execute=@n6.else, backpack_rotate=MakeRotator(Pitch=@n10, Yaw=@n9))
   n11 = Branch(execute=@n5.then_1, Condition=Get bIsInFrontEnd())
   > Set backpack_translation(backpack_translation="0,-0.5,3.5")
   Set backpack_translation(
    execute=@n11.else
    backpack_translation=MakeVector(
     Y=MapRangeClamped(Value=@n8.Yaw, InRangeA=-10, InRangeB=-30, OutRangeA=1, OutRangeB=-1)
     Z=MapRangeClamped(Value=@n8.Yaw, InRangeA=0, InRangeB=-30, OutRangeB=0)
    )
   )
   n12 = Event BlueprintProcessAnimInputEvent()
   > n13 = Branch(Condition=IsNamed(self=@n12.AnimInputEvent, InEventName=aie.Player.Zipline.Entered))
   > Set isZiplining(isZiplining:bool=true) -> (Output_Get:bool)
   Branch(execute=@n13.else, Condition=IsNamed(self=@n12.AnimInputEvent, InEventName=aie.Player.Zipline.Exited))
   > Set isZiplining(isZiplining=false)
   Event OnBeginSkydiving()
   > Set rotateback(rotateback:bool=true) -> (Output_Get:bool)
   Event OnEndSkydiving()
   > Set rotateback(rotateback=false)
   TryGetPawnOwner()
  ]}
]
variables [
 backpack_rotate {type=Rotator}
 backpack_translation {type=Vector}
 isIDLE {type=boolean}
 isZiplining {type=boolean}
 isZipplingAndRailing {type=boolean}
 rotateback {type=boolean}
]
